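# Physics asset for the male medium BunnySuit 01 player body cosmetic, built around two grenade attachments on dedicated dyn_ bones.
# Capsule bodies cover only those grenades plus spine_04, the right clavicle and the right upper arm, rather than a full ragdoll.
# Each grenade has tight angular limits, one constrained to the clavicle and the other to the spine.
# The spine, clavicle and upper arm form a chain with free swing and twist.
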PhysicsAsset /BRCosmetics/Characters/Player/Male/Medium/Bodies/M_MED_BunnySuit_01/Meshes/M_MED_Bunnysuit_01_Physics
bodies {
 clavicle_r [{shapeType=Capsule, length=12.9817, radius=6.3983, center=<-8.9291 0 0>, rotation=<rot 90 0 0>}]
 dyn_grenade_01 [{shapeType=Capsule, length=4.3789, radius=3.966, center=<0.6117 3.9069 1.3545>, rotation=<rot -19.392 -90 13.329>}]
 dyn_grenade_02 [{shapeType=Capsule, length=3.3782, radius=3.7505, center=<0.9122 3.0678 -0.7135>, rotation=<rot -6.3347 -90 12.1408>}]
 spine_04 [{shapeType=Capsule, length=11.9425, radius=7.7839}]
 upperarm_r [{shapeType=Capsule, length=18.3269, radius=4.4156, center=<-10.1155 1.8997 0.0002>, rotation=<rot 90 0 0>}]
}
constraints [
 {bone1Name=dyn_grenade_01, bone2Name=clavicle_r, swing1LimitDegrees=10, swing2LimitDegrees=1, twistLimitDegrees=10}
 {bone1Name=dyn_grenade_02, bone2Name=spine_04, swing1LimitDegrees=10, swing2LimitDegrees=1, twistLimitDegrees=13}
 {
  bone1Name=upperarm_r
  bone2Name=clavicle_r
  swing1LimitDegrees=45
  swing1Motion=Free
  swing2LimitDegrees=45
  swing2Motion=Free
  twistLimitDegrees=45
  twistMotion=Free
 }
 {
  bone1Name=clavicle_r
  bone2Name=spine_04
  swing1LimitDegrees=45
  swing1Motion=Free
  swing2LimitDegrees=45
  swing2Motion=Free
  twistLimitDegrees=45
  twistMotion=Free
 }
]
